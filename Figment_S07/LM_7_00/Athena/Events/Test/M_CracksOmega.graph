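# M_CracksOmega is a Default Lit material in an Athena event test folder that, by its names, puts cracks on grass landscape ground.
# It uses material attributes, feeding the Athena grass function MF_Athena_Grass_01 into MF_Cracks for the final result.
# All surface logic lives in those two functions, and the material is flagged for use with spline meshes.
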Material /Figment_S07/LM_7_00/Athena/Events/Test/M_CracksOmega
graph {
 properties {bUseMaterialAttributes=true, bUsedWithSplineMeshes=true, pixelDepthOffsetMode=PDOM_Legacy, usedShadingModels=Default Lit}
 nodes [MaterialFunctionCall!(In=MaterialFunctionCall() {materialFunction=/Figment_S07/LM_7_00/Athena/Environments/Landscape/MaterialFunctions/Standard/MF_Athena_Grass_01}.Result) {materialFunction=&/MF_Cracks}]
}
